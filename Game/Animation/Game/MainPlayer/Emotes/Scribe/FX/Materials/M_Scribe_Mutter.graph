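# Translucent unlit Niagara sprite material for the Scribe emote's mutter effect, drawing a grayscale character texture with animated wobble.
# A panning noise texture drives center-based UV scaling of the main texture, with dynamic parameters controlling scale range, speed and time offset.
# Smoothstepped texture values blend particle color into an inner color, optionally tinted by time of day, through the shared emissive function.
# A second smoothstep times particle alpha forms the opacity mask.
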Material /Game/Animation/Game/MainPlayer/Emotes/Scribe/FX/Materials/M_Scribe_Mutter
graph {
 properties {
  bOutputTranslucentVelocity=true
  bUsedWithNiagaraSprites=true
  blendMode=BLEND_Translucent
  opacityMaskClipValue=0.02
  pixelDepthOffsetMode=PDOM_Legacy
  shadingModel=MSM_Unlit
  usedShadingModels=Unlit
 }
 nodes [
  n1 = VectorParameter() {parameterName=Inner Color, defaultValue=<rgba 0.036 0 0.0573 1>}
  n2 = StaticSwitchParameter(
   True=MaterialFunctionCall(
    Color_Morning=VectorParameter() {parameterName=Color_Morning, defaultValue=<rgba 1 1 1 0>}.RGB
    Color_Day=VectorParameter() {parameterName=Color_Day, defaultValue=<rgba 1 1 1 0>}.RGB
    Color_Evening=VectorParameter() {parameterName=Color_Evening, defaultValue=<rgba 1 1 1 0>}.RGB
    Color_Night=VectorParameter() {parameterName=Color_Night, defaultValue=<rgba 1 1 1 0>}.RGB
   ) {materialFunction=&Color_Time_of_Day}.Result
   False=Constant3Vector() {constant=<rgba 1 1 1 0>}
  ) {parameterName=Use TOD}
  n3 = ScalarParameter() {defaultValue=0.2, parameterName=Smoothstep_Min}
  n4 = ScalarParameter() {defaultValue=0.4, parameterName=Smoothstep_Max}
  n5 = DynamicParameter() {paramNames [OffsetStrenght, Speed, Timeoffset, Param4]}
  n6 = LinearInterpolate(A=Constant2Vector() {g=0.5, r=0.5}, Alpha=@n5.OffsetStrenght)
  n7 = LinearInterpolate(A=Constant2Vector() {g=1.5, r=1.5}, Alpha=@n5.OffsetStrenght)
  n8 = Time()
  n9 = Add(A=@n8, B=@n5.Timeoffset) {constB=0.5}
  n10 = NamedRerouteUsage()
  n11 = Constant2Vector() {g=1, r=1}
  n12 = TextureSample(UVs=Panner(Coordinate=TextureCoordinate() {vTiling=0.5}, Time=LinearInterpolate(A=@n8, B=@n9, Alpha=@n10), Speed=Multiply(A=@n11, B=@n5.Speed)) {speedX=0.5, speedY=0.5}) {samplerType=SAMPLERTYPE_LinearColor, texture=/Game/Animation/Game/MainPlayer/Emotes/Scribe/FX/Textures/T_Emote_Scribe_N}
  n13 = TextureSampleParameter2D(UVs=MaterialFunctionCall(UVs=TextureCoordinate(), Texture Scale=LinearInterpolate(A=@n6, B=@n7, Alpha=@n12.R)) {materialFunction=&ScaleUVsByCenter}.UVs) {
   parameterName=Main Texture
   samplerType=SAMPLERTYPE_LinearGrayscale
   texture=/Game/Animation/Game/MainPlayer/Emotes/Scribe/FX/Textures/T_Scribe_Char_DF_M
  }
  MaterialFunctionCall(Emissive=LinearInterpolate(A=ParticleColor().RGB, B=Multiply(A=@n1.RGB, B=@n2), Alpha=SmoothStep(Min=@n3, Max=@n4, Value=@n13.R))) {materialFunction=&MF_FX_EmissiveAdjustments}
  Multiply(
   A=SmoothStep(
    Min=ScalarParameter() {parameterName=Alpha_SmoothStep_Min}
    Max=ScalarParameter() {defaultValue=0.2, parameterName=Alpha_SmoothStep_Max}
    Value=@n13.R
   ) {constMax=0.2}
   B=ParticleColor().A
  )
 ]
}
