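# Translucent unlit Niagara material for the main player's Scribe emote, drawing two animated "mutter" glyphs (Bu, Tsu).
# Vertex color R and G select cells of a two-column distance-field texture, smoothstepped into opacity and a particle-color to dark-purple emissive blend.
# A panning noise texture jitters each glyph's UV scale about center.
# Particle dynamic parameters drive that distortion plus vertex offsets for size, side separation and sine wobble.
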Material /Game/Animation/Game/MainPlayer/Emotes/Scribe/FX/Materials/M_Scribe_Mutter_v2
graph {
 properties {
  bOutputTranslucentVelocity=true
  bUsedWithNiagaraMeshParticles=true
  bUsedWithNiagaraSprites=true
  blendMode=BLEND_Translucent
  opacityMaskClipValue=0.02
  pixelDepthOffsetMode=PDOM_Legacy
  shadingModel=MSM_Unlit
  usedShadingModels=Unlit
 }
 nodes [
  n1 = ObjectPositionWS()
  n2 = WorldPosition()
  n3 = DynamicParameter() {parameterIndex=1, defaultValue=<rgba 1 1 22 1>, paramNames [Bu-Size,Tsu-Size,SideOffset,Param4]}
  n4 = LinearInterpolate(A=@n3.Bu-Size, B=@n3.Bu-Size, Alpha=NamedRerouteUsage()) {constA=0.6}
  n5 = DynamicParameter() {parameterIndex=2, paramNames [WobbleFrequency, WobbleAmplitude, WobbletimeOffset, Param4]}
  n6 = Add(
   A=Add(A=Time(), B=@n5.WobbletimeOffset)
   B=MaterialFunctionCall(Float2=TextureCoordinate() {coordinateIndex=1}) {materialFunction=&BreakOutFloat2Components}.R
  )
  n7 = Multiply(A=Sine(None=Multiply(A=@n6, B=@n5.WobbleFrequency)), B=@n5.WobbleAmplitude)
  n8 = Constant3Vector() {constant=<rgba 0 0 1 0>}
  n9 = Constant()
  n10 = DynamicParameter() {parameterIndex=1, defaultValue=<rgba 1 1 22 1>, paramNames [Bu-Size,Tsu-Size,SideOffset,Param4]}
  n11 = Transform(None=MaterialFunctionCall(X=@n9, Y=@n10.SideOffset, Z=@n9) {materialFunction=&MakeFloat3}.Result) {transformSourceType=TRANSFORMSOURCE_Local}
  n12 = Multiply(A=@n11) {constB=-1}
  n13 = NamedRerouteUsage()
  Add(
   A=Multiply(A=Subtract(A=@n1, B=@n2.XYZ), B=OneMinus(None=@n4)) {constB=0.5}
   B=Add(A=Multiply(A=@n7, B=@n8), B=LinearInterpolate(A=@n11, B=@n12, Alpha=@n13))
  )
  n14 = Constant() {r=0.3}
  n15 = Constant() {r=0.5}
  n16 = VertexColor()
  n17 = TextureCoordinate()
  n18 = Constant() {r=2}
  n19 = Constant() {r=1}
  n20 = AppendVector(A=@n18, B=@n19)
  n21 = AppendVector(A=Multiply(A=@n18, B=@n19), B=@n19)
  n22 = Floor(None=Multiply(A=@n21, B=Constant()))
  n23 = DynamicParameter() {paramNames [OffsetStrenght, Speed, Timeoffset, Param4]}
  n24 = TextureCoordinate()
  n25 = Time()
  n26 = LinearInterpolate(A=@n25, B=Add(A=@n25, B=@n23.Timeoffset) {constB=0.5}, Alpha=NamedRerouteUsage())
  n27 = Multiply(A=Constant2Vector() {g=1, r=1}, B=@n23.Speed)
  n28 = LinearInterpolate(
   A=LinearInterpolate(A=Constant2Vector() {g=0.5, r=0.5}, Alpha=@n23.OffsetStrenght)
   B=LinearInterpolate(A=Constant2Vector() {g=1.5, r=1.5}, Alpha=@n23.OffsetStrenght)
   Alpha=TextureSample(UVs=Panner(Coordinate=@n24, Time=@n26, Speed=@n27) {speedX=0.5, speedY=0.5}) {samplerType=SAMPLERTYPE_LinearColor, texture=/Game/Animation/Game/MainPlayer/Emotes/Scribe/FX/Textures/T_Emote_Scribe_N}.R
  )
  n29 = MaterialFunctionCall(UVs=Add(A=Divide(A=@n17, B=@n20), B=Divide(A=@n22, B=@n20)), Texture Scale=@n28) {materialFunction=&ScaleUVsByCenter}
  n30 = Floor(None=Multiply(A=@n21, B=Constant() {r=0.5}))
  n31 = MaterialFunctionCall(UVs=Add(A=Divide(A=@n17, B=@n20), B=Divide(A=@n30, B=@n20)), Texture Scale=@n28) {materialFunction=&ScaleUVsByCenter}
  n32 = Add(
   A=Multiply(
    A=@n16.R
    B=TextureSample(UVs=@n29.UVs) {samplerType=SAMPLERTYPE_LinearGrayscale, texture=/Game/Animation/Game/MainPlayer/Emotes/Scribe/FX/Textures/T_Scribe_Char_DF_M}.R
   )
   B=Multiply(
    A=@n16.G
    B=TextureSample(UVs=@n31.UVs) {samplerType=SAMPLERTYPE_LinearGrayscale, texture=&T_Scribe_Char_DF_M}.G
   )
  )
  MaterialFunctionCall(Emissive=LinearInterpolate(A=ParticleColor().RGB, B=Constant3Vector() {constant=<rgba 0.036 0 0.0573 0>}, Alpha=SmoothStep(Min=@n14, Max=@n15, Value=@n32))) {materialFunction=&MF_FX_EmissiveAdjustments}
  Multiply(A=SmoothStep(Value=@n32) {constMax=0.2}, B=ParticleColor().A)
  NamedRerouteDeclaration(None=@n16.G) {name=GreenMask, nodeColor=<rgba 1 0.4922 0 1>}
  NamedRerouteDeclaration(None=@n16.R) {name=RedMask, nodeColor=<rgba 0.2187 1 0 1>}
 ]
}
