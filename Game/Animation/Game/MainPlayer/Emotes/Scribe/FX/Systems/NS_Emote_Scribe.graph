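# Niagara system for the player's Scribe emote that emits fading, growing character sprites drifting outward to both sides of the emote.
# CharactersSprite, the only enabled emitter, bursts sprites in four partitions that pick sub-image and direction along a user vector, with sine bobbing and interpolated motion.
# Camera-to-vector dot products drive per-partition pivot and time offsets, and the system Timeline scales alpha.
# Mesh, previous-sprite and debug emitters are disabled.
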NiagaraSystem /Game/Animation/Game/MainPlayer/Emotes/Scribe/FX/Systems/NS_Emote_Scribe
dependencies {
 usedDataInterfaces [
  &NiagaraDataInterfaceCurve
  &NiagaraDataInterfaceArrayFloat
  &NiagaraDataInterfaceCamera
  &NiagaraDataInterfaceArrayFloat2
 ]
 usedDynamicInputs [&VectorLength, &SetBoolByFloatComparison]
 usedModules [
  &SystemState
  &self:SystemSpawnScript.SystemScriptSource.SystemScriptGraph.NiagaraNodeAssignment_0.SetVariables_#1
  &Timeline
 ]
}
emitters {
 Characters {
  bEnabled=false
  emitterUpdateScript [
   &EmitterState(Inactive Response=::NewEnumerator1, Life Cycle Mode=::NewEnumerator1, Loop Duration=0.65, UseLoopDelay=false)
   &SpawnBurst_Instantaneous(Use Spawn Probability=false, UseLoopCountLimit=false)
   &SpawnRate(SpawnRate=2.25, Use Spawn Probability=true) {enabled=false}
   /SetVariables(Emitter.SourcePosition:NiagaraPosition=&AddVectorToPosition(
     Position=$Engine.Emitter.SimulationPosition
     Vector=&TransformVector(Vector=<0 16 10>, Destination Space=::Simulation)
    )) {isSetParameters=true, module=SetVariables_#2}
  ]
  particleSpawnScript [
   &InitializeParticle(
    Color Mode=::NewEnumerator0
    Lifetime=4
    Mesh Scale=<1 0.12 0.09>
    Mesh Scale Mode=::NewEnumerator4
    Sprite Size=<16 9>
    Sprite Size Mode=::NewEnumerator4
    UsePositionOffset=false
    Write Material Random=true
   )
   &PartitionParticles()
   &DynamicMaterialParameters(
    Index 0 Param 1=&RandomRangeFloat(Minimum=0.6, Maximum=0.7)
    Index 0 Param 2=&RandomRangeFloat(Minimum=0.4, Maximum=0.5)
    Index 0 Param 3=&RandomRangeFloat(Minimum=0.5, Maximum=0.7)
    Index 1 Param 1=&RandomRangeFloat(Minimum=1, Maximum=0.7)
    Index 1 Param 2=&RandomRangeFloat(Minimum=1, Maximum=0.7)
    Index 1 Param 3=23
    Module.Param0WriteEnabled=true
    Module.Param0WriteEnabled1=true
    Module.Param1WriteEnabled=true
    Module.Param1WriteEnabled1=true
    Module.Param2WriteEnabled=true
    Module.Param2WriteEnabled1=true
    Module.Param3WriteEnabled=true
    Module.Param3WriteEnabled1=true
    Write Parameter Index 1=true
   )
   &CameraOffset(Camera Offset Amount=7)
  ]
  particleUpdateScript [
   &ParticleState()
   &ScaleMeshSize(Scale Factor=&Multiply_VectorByFloat(
     Vector=<1 1 1>
     Float=&FloatFromCurve(FloatCurve=DataInterface() {propertyValues {Curve {keys [
          {arriveTangent=2.8669, interpMode=RCIM_Cubic, leaveTangent=2.8669, leaveTangentWeight=0.3333, tangentMode=RCTM_User, time=0, value=0}
          {
           arriveTangent=0.7817
           arriveTangentWeight=0.1167
           interpMode=RCIM_Cubic
           leaveTangent=0.7817
           leaveTangentWeight=0.2167
           tangentMode=RCTM_User
           time=0.25
           value=1
          }
          <at 1 1.25>
         ]}}})
    ))
   &ScaleSpriteSize(Uniform Curve Sprite Scale=DataInterface() {propertyValues {Curve {keys [
        {arriveTangent=2.8669, interpMode=RCIM_Cubic, leaveTangent=2.8669, tangentMode=RCTM_User, time=0, value=0}
        {arriveTangent=0.7817, interpMode=RCIM_Cubic, leaveTangent=0.7817, tangentMode=RCTM_User, time=0.35, value=1}
        <at 1 1.25>
       ]}}})
   &ScaleColor(
    Scale Alpha=&FloatFromCurve(
     FloatCurve=DataInterface() {propertyValues {Curve {keys [
         <at 0.125 0 RCIM_Cubic>
         {arriveTangent=-0.0321, leaveTangent=-0.0321, time=0.25, value=1}
         {arriveTangent=-1.3333, interpMode=RCIM_Cubic, leaveTangent=-1.3333, time=0.8, value=1}
         {interpMode=RCIM_Cubic, leaveTangent=-4.8377, time=1, value=0}
        ]}}}
     Scale Curve=$System.Timeline.Float
    )
    ScaleA=true
    ScaleRGB=true
   )
   /SetVariables(
    Particles.LocalXOffset:NiagaraFloat=&Multiply_Float(
     A=&SelectFloatFromArray(Direct Array Index=$Particles.PartitionParticles.Partition, Float Selection Array=DataInterface() {propertyValues {FloatData [55,-55]}})
     B=&FloatFromCurve(FloatCurve=DataInterface() {propertyValues {Curve {keys [
          {arriveTangent=0.3514, interpMode=RCIM_Cubic, leaveTangent=0.3514, tangentMode=RCTM_User, time=0, value=0.25}
          {arriveTangent=1.0473, interpMode=RCIM_Cubic, tangentMode=RCTM_Break, time=1, value=1}
         ]}}})
    )
    Particles.LocalYOffset:NiagaraFloat=&Sine(
     Normalized Angle=&Add_Float(
      A=$Particles.Age
      B=&Add_Float(
       A=$Engine.Time
       B=&SelectFloatFromArray(Direct Array Index=$Particles.PartitionParticles.Partition, Float Selection Array=DataInterface() {propertyValues {FloatData [0,1.666]}})
      )
     )
     Period=4
     Scale=&Lerp_Float(A=1, B=3, Alpha=$Particles.NormalizedAge)
     Bias=0
    )
    Particles.LocalZOffset:NiagaraFloat=&Sine(
     Normalized Angle=&Add_Float(
      A=$Particles.Age
      B=&Add_Float(
       A=$Engine.Time
       B=&SelectFloatFromArray(Direct Array Index=$Particles.PartitionParticles.Partition, Float Selection Array=DataInterface() {propertyValues {FloatData [0,1.666]}})
      )
     )
     Period=4
     Scale=&Lerp_Float(A=1, B=3, Alpha=$Particles.NormalizedAge)
     Bias=0
    )
   ) {isSetParameters=true, module=SetVariables_#3}
   /SetVariables(Particles.TargetPosition:Vector3f=&AddVectorToPosition(
     Position=$Emitter.SourcePosition
     Vector=&TransformVector(
      Vector=&MakeVector(X=$Particles.LocalXOffset, Y=$Particles.LocalYOffset, Z=$Particles.LocalZOffset)
      Destination Space=::World
     )
    )) {isSetParameters=true, module=SetVariables_#4}
   &InterpolateOverTime(
    Initial Vector Value:NiagaraPosition=$Particles.Initial.Position
    Rate Of Change:NiagaraFloat=&Lerp_Float(A=5, B=3, Alpha=$Particles.NormalizedAge)
    Target Vector Value:Vector3f=$Particles.TargetPosition
    Variable Type=::NewEnumerator2
   )
   /SetVariables(Particles.Position:Vector3f=$Particles.InterpolateOverTime.InterpolateOverTime_V3_.Moving Average) {isSetParameters=true, module=SetVariables_#5}
   &DynamicMaterialParameters(
    Index 2 Param 1=0.25
    Index 2 Param 2=&Lerp_Float(A=2, B=5, Alpha=$Particles.NormalizedAge)
    Index 2 Param 3=&Add_Float(A=$Particles.Age, B=&Add_Float(A=$Engine.Time, B=1.66))
    Module.Param0WriteEnabled2=true
    Module.Param1WriteEnabled2=true
    Module.Param2WriteEnabled2=true
    Module.Param3WriteEnabled2=false
    Write Parameter Index 0=false
    Write Parameter Index 2=true
   ) {module=DynamicMaterialParameters001}
  ]
  rendererClasses [&NiagaraMeshRendererProperties]
  renderers [{rendererClass=&NiagaraMeshRendererProperties, rendererIndex=0}]
 }
 CharactersSprite {
  emitterUpdateScript [
   &EmitterState(
    Delay First Loop Only=true
    Inactive Response=::NewEnumerator1
    Life Cycle Mode=::NewEnumerator1
    Loop Delay=&FN_FrontEndFloatModifier(Gameplay=0, Front End=0.3)
    Loop Duration=0.333
    UseLoopDelay=true
   )
   &SpawnBurst_Instantaneous(Spawn Count=2, Use Spawn Probability=false, UseLoopCountLimit=false)
   /SetVariables(Emitter.SourcePosition:NiagaraPosition=&AddVectorToPosition(
     Position=$Engine.Emitter.SimulationPosition
     Vector=&TransformVector(Vector=<6 -7 9>, Destination Space=::Simulation)
    )) {isSetParameters=true, module=SetVariables_#2}
  ]
  particleSpawnScript [
   &PartitionParticles(Number Of Partitions=4)
   &InitializeParticle(
    Color Mode=::NewEnumerator0
    Lifetime=3
    Sprite Rotation Mode=::NewEnumerator0
    Sprite Size=&Multiply_Vector2D_ByFloat(Float=&RandomRangeFloat(Minimum=8, Maximum=6), Vector2D=<1 0.9>)
    Sprite Size Mode=::NewEnumerator4
    UsePositionOffset=false
    Write Material Random=true
   )
   /SetVariables(Particles.SubImageIndex=&SelectFloatFromArray(
     Direct Array Index=$Particles.PartitionParticles.Partition
     Direct Array Mode=::NewEnumerator1
     Float Selection Array=DataInterface() {propertyValues {FloatData [0,1,1,0]}}
    )) {isSetParameters=true, module=SetVariables_#6}
  ]
  particleUpdateScript [
   &ParticleState()
   &ScaleSpriteSize(Uniform Curve Sprite Scale=DataInterface() {propertyValues {Curve {keys [
        {arriveTangent=0.12, interpMode=RCIM_Cubic, leaveTangent=0.12, tangentMode=RCTM_User, time=0, value=0}
        {arriveTangent=1.3207, interpMode=RCIM_Cubic, leaveTangent=1.3207, tangentMode=RCTM_User, time=0.35, value=1}
        <at 1 1.25>
       ]}}})
   &ScaleColor(
    Scale Alpha=&FloatFromCurve(
     FloatCurve=DataInterface() {propertyValues {Curve {keys [
         <at 0.15 0 RCIM_Cubic>
         {arriveTangent=-0.0321, time=0.35, value=1}
         {arriveTangent=-1.5385, interpMode=RCIM_Cubic, leaveTangent=-1.5385, time=0.7, value=1}
         {interpMode=RCIM_Cubic, leaveTangent=-4.8377, time=1, value=0}
        ]}}}
     Scale Curve=$System.Timeline.Float
    )
    ScaleA=true
    ScaleRGB=true
   )
   &DynamicMaterialParameters(
    Index 0 Param 1=&FloatFromCurve(FloatCurve=DataInterface() {propertyValues {Curve {keys [{arriveTangent=-0.5168, interpMode=RCIM_Cubic, leaveTangent=-0.5168, tangentMode=RCTM_User, time=0, value=0.9}, <at 1 0.5 RCIM_Cubic>]}}})
    Index 0 Param 2=0.5
    Index 0 Param 3=&RandomRangeFloat(Minimum=0, Maximum=1)
    Module.Param0WriteEnabled=true
    Module.Param1WriteEnabled=true
    Module.Param2WriteEnabled=true
    Module.Param3WriteEnabled=true
   )
   /SetVariables(
    Particles.DirectionDotPivotOffset:NiagaraFloat=&DotProduct(
     Vector A=&NormalizeVector(
      Vector To Normalize=&Multiply_Vector(A=&ReturnCameraVector(Camera Vectors=::NewEnumerator1, Coordinate Space=::World), B=<1 1 0>)
      Safe Normalize=true
      Safe Fallback Threshold=0
      Safe Fallback Vector=<0 0 0.001>
     )
     Vector B=&NormalizeVector(
      Vector To Normalize=&Multiply_Vector(A=&TransformVector(Vector=$User.LocalHorizontalVector, Destination Space=::World), B=<1 1 0>)
      Safe Normalize=true
      Safe Fallback Threshold=0
      Safe Fallback Vector=<0 0 0.001>
     )
    )
    Particles.DirectionDotTimeOffset:NiagaraFloat=&DotProduct(
     Vector A=&NormalizeVector(
      Vector To Normalize=&Multiply_Vector(A=&ReturnCameraVector(Camera Vectors=::NewEnumerator3, Coordinate Space=::World), B=<1 1 0>)
      Safe Normalize=true
      Safe Fallback Threshold=0
      Safe Fallback Vector=<0 0 0.001>
     )
     Vector B=&NormalizeVector(
      Vector To Normalize=&Multiply_Vector(A=&TransformVector(Vector=$User.LocalHorizontalVector, Destination Space=::World), B=<1 1 0>)
      Safe Normalize=true
      Safe Fallback Threshold=0
      Safe Fallback Vector=<0 0 0.001>
     )
    )
   ) {isSetParameters=true, module=SetVariables_#7}
   /SetVariables(
    Particles.PivotOffset:Vector2f=&Multiply_Vector2D_ByFloat(
     Float=&Multiply_Float(A=&Abs_Float(Float:NiagaraFloat=$Particles.DirectionDotPivotOffset), B=$User.PivotOffset)
     Vector2D=&SelectVector2DFromArray(
      Direct Array Index=$Particles.PartitionParticles.Partition
      Direct Array Mode=::NewEnumerator1
      Vector 2D Selection Array=DataInterface() {propertyValues {FloatData [<1 0>, <-1 0>, <-1 0>, <1 0>]}}
     )
    )
    Particles.TimeOffset:NiagaraFloat=&Multiply_Float(
     A=$User.TimeOffset
     B=&Multiply_Float(
      A=$Particles.DirectionDotTimeOffset
      B=&SelectFloatFromArray(
       Direct Array Index=$Particles.PartitionParticles.Partition
       Direct Array Mode=::NewEnumerator1
       Float Selection Array=DataInterface() {propertyValues {FloatData [1,-1,1,-1]}}
      )
     )
    )
   ) {isSetParameters=true, module=SetVariables_#8}
   /SetVariables(Particles.OffsetAge:NiagaraFloat=&Add_Float(A=$Particles.Age, B=$Particles.TimeOffset)) {isSetParameters=true, module=SetVariables_#9}
   /SetVariables(
    Particles.LocalXOffset:NiagaraFloat=&Multiply_Float(
     A=&SelectFloatFromArray(
      Direct Array Index=$Particles.PartitionParticles.Partition
      Direct Array Mode=::NewEnumerator1
      Float Selection Array=DataInterface() {propertyValues {FloatData [20,20,-20,-20]}}
     )
     B=$Particles.OffsetAge
    )
    Particles.LocalXYOffset:Vector3f=&Multiply_VectorByFloat(
     Vector=$User.LocalHorizontalVector
     Float=&Multiply_Float(
      A=$Particles.OffsetAge
      B=&Multiply_Float(
       A=$User.LocalHorizontalDistance
       B=&SelectFloatFromArray(Direct Array Index=$Particles.PartitionParticles.Partition, Float Selection Array=DataInterface() {propertyValues {FloatData [1,1,-1,-1]}})
      )
     )
    )
    Particles.LocalYOffset:NiagaraFloat=&Sine(Normalized Angle=$Particles.Age, Period=4, Scale=&Lerp_Float(A=4, B=0, Alpha=$Particles.NormalizedAge), Bias=0)
    Particles.LocalZOffset:NiagaraFloat=&Sine(
     Normalized Angle=&Add_Float(
      A=$Particles.OffsetAge
      B=&SelectFloatFromArray(
       Direct Array Index=$Particles.PartitionParticles.Partition
       Float Selection Array=DataInterface() {propertyValues {FloatData [1.5,1.5,1.5,1.5]}}
      )
     )
     Period=2
     Scale=&Lerp_Float(A=3, B=9, Alpha=$Particles.NormalizedAge)
     Bias=0
    )
   ) {isSetParameters=true, module=SetVariables_#3}
   /SetVariables(Particles.TargetPosition:Vector3f=&AddVectorToPosition(
     Position=$Emitter.SourcePosition
     Vector=&TransformVector(
      Vector=&Add_Vector(A=$Particles.LocalXYOffset, B=&MakeVector(X=0, Y=$Particles.LocalYOffset, Z=$Particles.LocalZOffset))
      Destination Space=::World
     )
    )) {isSetParameters=true, module=SetVariables_#4}
   &InterpolateOverTime(
    Initial Vector Value:NiagaraPosition=$Particles.Initial.Position
    Rate Of Change:NiagaraFloat=&Lerp_Float(A=5, B=3, Alpha=$Particles.NormalizedAge)
    Target Vector Value:Vector3f=$Particles.TargetPosition
    Variable Type=::NewEnumerator2
   )
   /SetVariables(Particles.Position:Vector3f=$Particles.InterpolateOverTime.InterpolateOverTime_V3_.Moving Average) {isSetParameters=true, module=SetVariables_#5}
   /SetVariables(Particles.SpriteRotation:NiagaraFloat=&Multiply_Float(
     A=&RandomRangeFloat(Minimum=10, Maximum=-10)
     B=&Sine(Normalized Angle=$Emitter.Age, Period=2, Scale=1, Bias=0)
    )) {isSetParameters=true, module=SetVariables_#10}
  ]
  rendererClasses [&NiagaraSpriteRendererProperties]
  renderers [{rendererClass=&NiagaraSpriteRendererProperties}]
 }
 CharactersSpritePrevious {
  bEnabled=false
  emitterUpdateScript [
   &EmitterState(
    Delay First Loop Only=true
    Inactive Response=::NewEnumerator1
    Life Cycle Mode=::NewEnumerator1
    Loop Delay=&FN_FrontEndFloatModifier(Gameplay=0, Front End=0.3)
    Loop Duration=0.333
    UseLoopDelay=true
   )
   &SpawnBurst_Instantaneous(Spawn Count=2, Use Spawn Probability=false, UseLoopCountLimit=false)
   /SetVariables(Emitter.SourcePosition:NiagaraPosition=&AddVectorToPosition(
     Position=$Engine.Emitter.SimulationPosition
     Vector=&TransformVector(Vector=<0 0 10>, Destination Space=::Simulation)
    )) {isSetParameters=true, module=SetVariables_#2}
  ]
  particleSpawnScript [
   &PartitionParticles(Number Of Partitions=4)
   &InitializeParticle(
    Color Mode=::NewEnumerator0
    Lifetime=3
    Sprite Rotation Mode=::NewEnumerator0
    Sprite Size=&Multiply_Vector2D_ByFloat(Float=&RandomRangeFloat(Minimum=8, Maximum=6), Vector2D=<1 0.9>)
    Sprite Size Mode=::NewEnumerator4
    UsePositionOffset=false
    Write Material Random=true
   )
   /SetVariables(Particles.SubImageIndex=&SelectFloatFromArray(
     Direct Array Index=$Particles.PartitionParticles.Partition
     Direct Array Mode=::NewEnumerator1
     Float Selection Array=DataInterface() {propertyValues {FloatData [0,1,1,0]}}
    )) {isSetParameters=true, module=SetVariables_#6}
  ]
  particleUpdateScript [
   &ParticleState()
   &ScaleSpriteSize(Uniform Curve Sprite Scale=DataInterface() {propertyValues {Curve {keys [
        {arriveTangent=0.12, interpMode=RCIM_Cubic, leaveTangent=0.12, tangentMode=RCTM_User, time=0, value=0}
        {arriveTangent=1.3207, interpMode=RCIM_Cubic, leaveTangent=1.3207, tangentMode=RCTM_User, time=0.35, value=1}
        <at 1 1.25>
       ]}}})
   &ScaleColor(
    Scale Alpha=&FloatFromCurve(
     FloatCurve=DataInterface() {propertyValues {Curve {keys [
         <at 0.15 0 RCIM_Cubic>
         {arriveTangent=-0.0321, time=0.35, value=1}
         {arriveTangent=-1.5385, interpMode=RCIM_Cubic, leaveTangent=-1.5385, time=0.7, value=1}
         {interpMode=RCIM_Cubic, leaveTangent=-4.8377, time=1, value=0}
        ]}}}
     Scale Curve=$System.Timeline.Float
    )
    ScaleA=true
    ScaleRGB=true
   )
   &DynamicMaterialParameters(
    Index 0 Param 1=&FloatFromCurve(FloatCurve=DataInterface() {propertyValues {Curve {keys [{arriveTangent=-0.5168, interpMode=RCIM_Cubic, leaveTangent=-0.5168, tangentMode=RCTM_User, time=0, value=0.9}, <at 1 0.5 RCIM_Cubic>]}}})
    Index 0 Param 2=0.5
    Index 0 Param 3=&RandomRangeFloat(Minimum=0, Maximum=1)
    Module.Param0WriteEnabled=true
    Module.Param1WriteEnabled=true
    Module.Param2WriteEnabled=true
    Module.Param3WriteEnabled=true
   )
   /SetVariables(Particles.DirectionDot:NiagaraFloat=&DotProduct(
     Vector A=&ReturnCameraVector(Camera Vectors=::NewEnumerator4, Coordinate Space=::World, Origin Position=$Particles.Position)
     Vector B=&TransformVector(Vector=<0 0 -1>, Destination Space=::Simulation)
    )) {isSetParameters=true, module=SetVariables_#7}
   /SetVariables(
    Particles.PivotOffset:Vector2f=&Multiply_Vector2D_ByFloat(
     Float=&OneMinusFloat(Float=&Abs_Float(Float:NiagaraFloat=$Particles.DirectionDot))
     Vector2D=&SelectVector2DFromArray(
      Direct Array Index=$Particles.PartitionParticles.Partition
      Direct Array Mode=::NewEnumerator1
      Vector 2D Selection Array=DataInterface() {propertyValues {FloatData [<0.333 0>, <-0.333 0>, <-0.333 0>, <0.333 0>]}}
     )
    )
    Particles.TimeOffset:NiagaraFloat=&Multiply_Float(
     A=0.1333
     B=&Multiply_Float(
      A=$Particles.DirectionDot
      B=&SelectFloatFromArray(
       Direct Array Index=$Particles.PartitionParticles.Partition
       Direct Array Mode=::NewEnumerator1
       Float Selection Array=DataInterface() {propertyValues {FloatData [1,-1,1,-1]}}
      )
     )
    )
   ) {isSetParameters=true, module=SetVariables_#8}
   /SetVariables(Particles.OffsetAge:NiagaraFloat=&Add_Float(A=$Particles.Age, B=$Particles.TimeOffset)) {isSetParameters=true, module=SetVariables_#9}
   /SetVariables(
    Particles.LocalXOffset:NiagaraFloat=&Multiply_Float(
     A=&SelectFloatFromArray(
      Direct Array Index=$Particles.PartitionParticles.Partition
      Direct Array Mode=::NewEnumerator1
      Float Selection Array=DataInterface() {propertyValues {FloatData [20,20,-20,-20]}}
     )
     B=$Particles.OffsetAge
    )
    Particles.LocalYOffset:NiagaraFloat=&Sine(
     Normalized Angle=&Add_Float(
      A=$Particles.Age
      B=&Add_Float(
       A=$Engine.Time
       B=&SelectFloatFromArray(
        Direct Array Index=$Particles.PartitionParticles.Partition
        Direct Array Mode=::NewEnumerator1
        Float Selection Array=DataInterface() {propertyValues {FloatData [0,0,0.222,0.222]}}
       )
      )
     )
     Period=4
     Scale=&Lerp_Float(A=1, B=3, Alpha=$Particles.NormalizedAge)
     Bias=0
    )
    Particles.LocalZOffset:NiagaraFloat=&Sine(
     Normalized Angle=&Add_Float(
      A=$Particles.OffsetAge
      B=&Add_Float(
       A=$Engine.Time
       B=&SelectFloatFromArray(
        Direct Array Index=$Particles.PartitionParticles.Partition
        Direct Array Mode=::NewEnumerator1
        Float Selection Array=DataInterface() {propertyValues {FloatData [0,0,0.222,0.222]}}
       )
      )
     )
     Period=2
     Scale=&Lerp_Float(A=3, B=9, Alpha=$Particles.NormalizedAge)
     Bias=0
    )
   ) {isSetParameters=true, module=SetVariables_#3}
   /SetVariables(Particles.TargetPosition:Vector3f=&AddVectorToPosition(
     Position=$Emitter.SourcePosition
     Vector=&TransformVector(
      Vector=&MakeVector(X=$Particles.LocalXOffset, Y=$Particles.LocalYOffset, Z=$Particles.LocalZOffset)
      Destination Space=::World
     )
    )) {isSetParameters=true, module=SetVariables_#4}
   &InterpolateOverTime(
    Initial Vector Value:NiagaraPosition=$Particles.Initial.Position
    Rate Of Change:NiagaraFloat=&Lerp_Float(A=5, B=3, Alpha=$Particles.NormalizedAge)
    Target Vector Value:Vector3f=$Particles.TargetPosition
    Variable Type=::NewEnumerator2
   )
   /SetVariables(Particles.Position:Vector3f=$Particles.InterpolateOverTime.InterpolateOverTime_V3_.Moving Average) {isSetParameters=true, module=SetVariables_#5}
   /SetVariables(Particles.SpriteRotation:NiagaraFloat=&Multiply_Float(
     A=&RandomRangeFloat(Minimum=10, Maximum=-10)
     B=&Sine(Normalized Angle=$Emitter.Age, Period=2, Scale=1, Bias=0)
    )) {isSetParameters=true, module=SetVariables_#10}
  ]
  rendererClasses [&NiagaraSpriteRendererProperties]
  renderers [{rendererClass=&NiagaraSpriteRendererProperties}]
 }
 FN_Debug {
  bEnabled=false
  emitterUpdateScript [
   &EmitterState(Life Cycle Mode=::NewEnumerator1, Loop Behavior=::NewEnumerator1, Loop Duration Mode=::NewEnumerator1, UseLoopDelay=false)
   &SpawnBurst_Instantaneous(Spawn Count=1, Use Spawn Probability=false)
  ]
  particleSpawnScript [/Niagara/Modules/Spawn/Initialization/InitializeParticle(
    Lifetime=5
    Module.Write Color=true
    Module.Write Lifetime=true
    Module.Write Mass=true
    Module.Write Material Random=true
    Module.Write Position=false
    Module.Write Scale=false
    Module.Write SpriteRotation=false
    Module.Write SpriteSize=true
    Position=&SimulationPosition()
    Sprite Size=<100 100>
   )]
  particleUpdateScript [
   &ParticleState(Kill Particles When Lifetime Has Elapsed=false)
   &FN_Timer()
   /SetVariables(Particles.Debug:NiagaraFloat=$Engine.System.Age) {isSetParameters=true}
   &DynamicMaterialParameters(
    Index 0 Param 1=$Particles.Debug
    Module.Param0WriteEnabled=true
    Module.Param1WriteEnabled=true
    Module.Param2WriteEnabled=true
    Module.Param3WriteEnabled=true
   )
   &CameraOffset(Camera Offset Amount=111)
  ]
  rendererClasses [&NiagaraSpriteRendererProperties]
  renderers [{rendererClass=&NiagaraSpriteRendererProperties}]
 }
}
summary {
 systemName=NS_Emote_Scribe
 emitters [
  {bEnabled=false, emitterName=Characters, rendererClasses [&NiagaraMeshRendererProperties]}
  {emitterName=CharactersSprite, rendererClasses [&NiagaraSpriteRendererProperties]}
  {bEnabled=false, emitterName=FN_Debug, rendererClasses [&NiagaraSpriteRendererProperties]}
  {bEnabled=false, emitterName=CharactersSpritePrevious, rendererClasses [&NiagaraSpriteRendererProperties]}
 ]
 userVariables [
  User.bIsFrontEnd {type=NiagaraBool}
  User.bIsFrontEndPreview {type=NiagaraBool}
  User.LocalHorizontalDistance {type=NiagaraFloat, defaultValue=20, description=""}
  User.LocalHorizontalVector {type=Vector3f, description="", defaultValue {struct=Vector3f, value=<1 0.066 -0.1>}}
  User.PivotOffset {type=NiagaraFloat, defaultValue=0.333, description=""}
  User.Skeletal Mesh {
   type=NiagaraDataInterfaceSkeletalMesh
   defaultValue {
    struct=NiagaraExt_VariableValue_DataInterface
    value {dataInterface=&self:NiagaraDataInterfaceSkeletalMesh_0}
   }
  }
  User.TimeOffset {type=NiagaraFloat, defaultValue=-0.133, description=""}
 ]
}
